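# Control Rig for the Armadillo Armored medium male player body that drives corrective deform bones at shoulders, elbows, wrists, pecs, hips and knees.
# Base bones blend between neighboring bones through constraints; spherical pose readers or remapped joint rotations weight offset transforms on their children.
# Each region is toggled by deform_bool, takes offsets from deform_value, and loops twice, mirroring left to right by name-token replacement.
# Lat and bicep logic references bones absent from this hierarchy.
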ControlRigBlueprint /Game/Characters/Player/Male/Medium/Bodies/M_MED_Armadillo_Armored/Meshes/M_MED_DeformRig_Armadillo_Armored_ControlRig
graphs [
 ContainedGraph {nodes [
   n1 = RigVMFunctionEntryNode()
   > RigVMFunctionReturnNode(
    OutIndex=@n1.Index
    OutItems=RigVMDispatchNode(
     Values=((),())
     Values.1=/Script/ControlRig.RigUnit_CollectionReplaceItemsArray(Items=@n1.Items, Old=@n1.l_side_token, New=@n1.r_side_token) {is=struct RigVMUnitNode}.Result
     Index=@n1.Index
     Values.0=@n1.Items
    ) {template="DISPATCH_RigVMDispatch_SelectInt32(in Index,in Values,out Result)"}.Result
   )
  ]}
 ContainedGraph {nodes [
   n1 = RigVMFunctionEntryNode()
   n2 = &RigUnit_GetTransform(Space=LocalSpace, Item=@n1.Item)
   n3 = &RigUnit_GetTransform(Space=LocalSpace, bInitial=true, Item=@n1.Item)
   > RigVMFunctionReturnNode(OutTransform=/Script/RigVM.RigVMFunction_MathTransformToEulerTransform(Value=&RigVMFunction_MathTransformMakeRelative(Global=@n2.Transform, Parent=@n3.Transform).Local) {is=struct RigVMUnitNode, template="Cast::Execute(in Value,out Result)"}.Result)
  ]}
 ContainedGraph {nodes [
   n1 = RigVMFunctionEntryNode()
   n2 = &RigVMFunction_MathTransformToEulerTransform(Value=@n1.transform) {template="Cast::Execute(in Value,out Result)"}
   > &RigUnit_ModifyTransforms(
    ItemToModify=((Item=(Type=Bone),Transform=(Rotation=(X=0,Y=0,Z=0,W=1),Translation=(X=0,Y=0,Z=0),Scale3D=(X=0.6,Y=1,Z=1))))
    ItemToModify.0.Transform.Rotation=/Script/RigVM.RigVMFunction_MathQuaternionFromRotator(Rotator=@n2.Result.Rotation) {is=struct RigVMUnitNode}.Result
    ItemToModify.0.Item=RigVMDispatchNode(
     Values=((Type=None,Name="None"),(Type=None,Name="None"))
     Values.1=/Script/ControlRig.RigUnit_ItemReplace(Item=@n1.item, New=@n1.r_side_token, Old=@n1.l_side_token) {is=struct RigVMUnitNode}.Result
     Index=@n1.index
     Values.0=@n1.item
    ) {template="DISPATCH_RigVMDispatch_SelectInt32(in Index,in Values,out Result)"}.Result
    ItemToModify.0.Transform.Translation=RigVMDispatchNode(
     Values=((X=0,Y=0,Z=0),(X=0,Y=0,Z=0))
     Values.1=/Script/RigVM.RigVMFunction_MathVectorScale(Factor=-1, Value=@n2.Result.Location) {is=struct RigVMUnitNode, template="Scale::Execute(in Value,in Factor,out Result)"}.Result
     Index=@n1.index
     Values.0=@n2.Result.Location
    ) {template="DISPATCH_RigVMDispatch_SelectInt32(in Index,in Values,out Result)"}.Result
    ItemToModify.0.Transform.Scale3D=@n2.Result.Scale
    Weight=@n1.weights
   )
   > RigVMFunctionReturnNode(outWeights=@n1.weights, outIndex=@n1.index)
  ]}
 ContainedGraph {nodes [
   n1 = RigVMFunctionEntryNode()
   n2 = RigVMDispatchNode(
    Values=((),())
    Values.1=&RigUnit_CollectionReplaceItemsArray(Items=@n1.masters, Old=@n1.l_side_token, New=@n1.r_side_token).Result
    Index=@n1.Index
    Values.0=@n1.masters
   ) {template="DISPATCH_RigVMDispatch_SelectInt32(in Index,in Values,out Result)"}
   > &RigUnit_ParentConstraint(
    AdvancedSettings=(RotationOrderForFilter=XZY)
    Parents=((Item=(Type=Bone),Weight=0.7),(Item=(Type=Bone),Weight=1))
    Parents.0.Item=RigVMDispatchNode(Index=0, Array=@n2.Result) {template="DISPATCH_RigVMDispatch_ArrayGetAtIndex(in Array,in Index,out Element)"}.Element
    Parents.1.Item=RigVMDispatchNode(Index=1, Array=@n2.Result) {template="DISPATCH_RigVMDispatch_ArrayGetAtIndex(in Array,in Index,out Element)"}.Element
    Parents.0.Weight=RigVMDispatchNode(Index=0, Array=@n1.weights) {template="DISPATCH_RigVMDispatch_ArrayGetAtIndex(in Array,in Index,out Element)"}.Element
    Parents.1.Weight=RigVMDispatchNode(Index=1, Array=@n1.weights) {template="DISPATCH_RigVMDispatch_ArrayGetAtIndex(in Array,in Index,out Element)"}.Element
    Child=RigVMDispatchNode(
     Values=((Type=None,Name="None"),(Type=None,Name="None"))
     Values.1=&RigUnit_ItemReplace(Item=@n1.slave, Old=@n1.l_side_token, New=@n1.r_side_token).Result
     Index=@n1.Index
     Values.0=@n1.slave
    ) {template="DISPATCH_RigVMDispatch_SelectInt32(in Index,in Values,out Result)"}.Result
    Weight=@n1.alpha
   )
   > RigVMFunctionReturnNode(outIndex=@n1.Index)
  ]}
 ContainedGraph {nodes [
   n1 = RigVMFunctionEntryNode()
   > n2 = &RigUnit_SphericalPoseReader(
    ActiveRegionScaleFactors=(PositiveWidth=3.5,NegativeWidth=0.5,PositiveHeight=3,NegativeHeight=3)
    Debug=(bDrawDebug=false,bDraw2D=false,bDrawLocalAxes=false,DebugScale=25,DebugSegments=20,DebugThickness=0.25)
    FalloffRegionScaleFactors=(PositiveWidth=1,NegativeWidth=1,PositiveHeight=1,NegativeHeight=1)
    RotationOffset=RigVMDispatchNode(Array=@n1.rotationOffset, Index=@n1.index) {template="DISPATCH_RigVMDispatch_ArrayGetAtIndex(in Array,in Index,out Element)"}.Element
    DriverItem=RigVMDispatchNode(
     Values=((Type=None,Name="None"),(Type=None,Name="None"))
     Values.1=&RigUnit_ItemReplace(Old=@n1.l_side_token, New=@n1.r_side_token, Item=@n1.driver).Result
     Index=@n1.index
     Values.0=@n1.driver
    ) {template="DISPATCH_RigVMDispatch_SelectInt32(in Index,in Values,out Result)"}.Result
    DriverAxis=RigVMDispatchNode(
     Values=((X=0,Y=0,Z=0),(X=0,Y=0,Z=0))
     Values.1=&RigVMFunction_MathVectorScale(Factor=-1, Value=@n1.driverAxis) {template="Scale::Execute(in Value,in Factor,out Result)"}.Result
     Values.0=@n1.driverAxis
     Index=@n1.index
    ) {template="DISPATCH_RigVMDispatch_SelectInt32(in Index,in Values,out Result)"}.Result
    OptionalParentItem=RigVMDispatchNode(
     Values=((Type=None,Name="None"),(Type=None,Name="None"))
     Values.1=&RigUnit_ItemReplace(Old=@n1.l_side_token, New=@n1.r_side_token, Item=@n1.parentSpace).Result
     Index=@n1.index
     Values.0=@n1.parentSpace
    ) {template="DISPATCH_RigVMDispatch_SelectInt32(in Index,in Values,out Result)"}.Result
    ActiveRegionScaleFactors.PositiveWidth=@n1.positiveWidth
    ActiveRegionSize=@n1.activeRegionSize
    FlipHeightScaling=@n1.flipHeightScaling
    FlipWidthScaling=@n1.flipWidthScaling
    Debug.bDrawDebug=@n1.drawDebug
    FalloffSize=@n1.FalloffSize
    ActiveRegionScaleFactors.NegativeHeight=@n1.negativeHeight
    ActiveRegionScaleFactors.PositiveHeight=@n1.positiveHeight
    ActiveRegionScaleFactors.NegativeWidth=@n1.negativeWidth
   )
   > RigVMFunctionReturnNode(outWeights=@n2.OutputParam, outIndex=@n1.index)
  ]}
 RigVMFunctionLibrary {nodes [
   GetDeltaTransformFromRefPose_FNC = RigVMCollapseNode(ExecuteContext=(), Item=(Type=Bone))
   GetMirrorItems_FNC = RigVMCollapseNode(ExecuteContext=(), Items=(), l_side_token=_l, r_side_token=_r)
   PoseReader_FNC = RigVMCollapseNode(
    ExecuteContext=()
    FalloffSize=0.2
    activeRegionSize=0.1
    driver=(Type=Bone)
    driverAxis=(X=1,Y=0,Z=0)
    l_side_token=_l
    negativeHeight=1
    negativeWidth=1
    parentSpace=(Type=Bone)
    positiveHeight=1
    positiveWidth=1
    r_side_token=_r
    rotationOffset=()
   )
   SetDrivenKey_FNC = RigVMCollapseNode(
    ExecuteContext=()
    item=(Type=Bone)
    l_side_token=_l
    r_side_token=_r
    transform=(Rotation=(X=0,Y=0,Z=0,W=1),Translation=(X=0,Y=0,Z=0),Scale3D=(X=1,Y=1,Z=1))
    weights=1
   )
   TwoParentConstraint_FNC = RigVMCollapseNode(ExecuteContext=(), alpha=1, l_side_token=_l, masters=(), r_side_token=_r, slave=(Type=Bone), weights=())
  ]}
 RigVMModel {nodes [
   RigUnit_BeginExecution = &RigUnit_BeginExecution()
   n1 = /Script/ControlRig.RigUnit_SequenceExecution(ExecuteContext=@RigUnit_BeginExecution.ExecutePin) {is=struct RigVMUnitNode}
   n2 = &RigUnit_SequenceExecution(ExecuteContext=@n1.A)
   Branch_4_RigVMUnitNode = &RigVMFunction_ControlFlowBranch(Condition=RigVMVariableNode() {variable=deform_bool}.Value.lat_15_#1, ExecuteContext=@n2.D)
   n3 = /Script/RigVM.RigVMFunction_ForLoopCount(Count=2, ExecutePin=@Branch_4_RigVMUnitNode.True) {is=struct RigVMUnitNode}
   > PoseReader_FNC_1_1_1_1_2 = RigVMFunctionReferenceNode(
    FalloffSize=0.2
    activeRegionSize=0.05
    drawDebug=false
    driver=(Type=Bone,Name="upperarm_l")
    driverAxis=(X=1,Y=0,Z=0)
    flipWidthScaling=false
    l_side_token=_l
    negativeHeight=4
    negativeWidth=0.5
    parentSpace=(Type=Bone,Name="spine_03")
    positiveHeight=1.5
    positiveWidth=4
    r_side_token=_r
    rotationOffset=((X=0,Y=-150,Z=0),(X=0,Y=30,Z=0))
    index=@n3.Index
    flipHeightScaling=/Script/RigVM.RigVMFunction_MathIntEquals(B=1, A=@n3.Index) {is=struct RigVMUnitNode}.Result
   )
   > SetDrivenKey_FNC_2_1_2_1_2 = RigVMFunctionReferenceNode(
    item=(Type=Bone,Name="deform_lat_l")
    l_side_token=_l
    r_side_token=_r
    index=@PoseReader_FNC_1_1_1_1_2.outIndex
    weights=@PoseReader_FNC_1_1_1_1_2.outWeights
    transform=RigVMVariableNode() {variable=deform_value}.Value.lat_45_#2
   )
   n4 = RigVMVariableNode() {variable=deform_bool}
   Branch_RigVMUnitNode = &RigVMFunction_ControlFlowBranch(
    Condition=&RigVMFunction_MathBoolOr(B=@n4.Value.bicep_13_#3, A=@n4.Value.elbow_3_#4).Result
    ExecuteContext=@n2.A
   )
   n5 = &RigVMFunction_ForLoopCount(Count=2, ExecutePin=@Branch_RigVMUnitNode.True)
   > GetMirrorItems_FNC_1_2_2_2_1 = RigVMFunctionReferenceNode(Items=((Type=Bone,Name="lowerarm_l"),(Type=Bone,Name="upperarm_l")), l_side_token=_l, r_side_token=_r, Index=@n5.Index)
   > GetDeltaTransformFromRefPose_FNC_1 = RigVMFunctionReferenceNode(Item=RigVMDispatchNode(Index=0, Array=@GetMirrorItems_FNC_1_2_2_2_1.OutItems) {template="DISPATCH_RigVMDispatch_ArrayGetAtIndex(in Array,in Index,out Element)"}.Element)
   n6 = &RigUnit_SequenceExecution(ExecuteContext=@GetDeltaTransformFromRefPose_FNC_1.ExecuteContext)
   Branch_6_RigVMUnitNode = &RigVMFunction_ControlFlowBranch(ExecuteContext=@n6.A, Condition=RigVMVariableNode() {variable=deform_bool}.Value.elbow_3_#4)
   n7 = &RigUnit_SequenceExecution(ExecuteContext=@Branch_6_RigVMUnitNode.True)
   TwoParentConstraint_FNC_3_2 = RigVMFunctionReferenceNode(
    alpha=1
    l_side_token=_l
    masters=((Type=Bone,Name="lowerarm_l"),(Type=Bone,Name="upperarm_l"))
    r_side_token=_r
    slave=(Type=Bone,Name="deform_elbow_in_base_l")
    weights=(0.4,0.6)
    ExecuteContext=@n7.A
    Index=@GetMirrorItems_FNC_1_2_2_2_1.OutIndex
   )
   n8 = &RigVMFunction_MathFloatRemap(SourceMaximum=90, SourceMinimum=15, bClamp=true, Value=@GetDeltaTransformFromRefPose_FNC_1.OutTransform.Rotation.Yaw)
   > SetDrivenKey_FNC_3_2 = RigVMFunctionReferenceNode(
    item=(Type=Bone,Name="deform_elbow_in_l")
    l_side_token=_l
    r_side_token=_r
    index=@TwoParentConstraint_FNC_3_2.outIndex
    transform=RigVMVariableNode() {variable=deform_value}.Value.elbow_in_24_#5
    weights=@n8.Result
   )
   TwoParentConstraint_FNC_3_2_1 = RigVMFunctionReferenceNode(
    alpha=1
    l_side_token=_l
    masters=((Type=Bone,Name="lowerarm_l"),(Type=Bone,Name="upperarm_l"))
    r_side_token=_r
    slave=(Type=Bone,Name="deform_elbow_out_base_l")
    weights=(0.3,0.7)
    ExecuteContext=@n7.B
    Index=@GetMirrorItems_FNC_1_2_2_2_1.OutIndex
   )
   > SetDrivenKey_FNC_3_2_1 = RigVMFunctionReferenceNode(
    item=(Type=Bone,Name="deform_elbow_out_l")
    l_side_token=_l
    r_side_token=_r
    index=@TwoParentConstraint_FNC_3_2_1.outIndex
    transform=RigVMVariableNode() {variable=deform_value}.Value.elbow_out_26_#6
    weights=@n8.Result
   )
   Branch_7_RigVMUnitNode = &RigVMFunction_ControlFlowBranch(ExecuteContext=@n6.B, Condition=RigVMVariableNode() {variable=deform_bool}.Value.bicep_13_#3)
   SetDrivenKey_FNC_3_2_2 = RigVMFunctionReferenceNode(
    item=(Type=Bone,Name="deform_bicep_l")
    l_side_token=_l
    r_side_token=_r
    transform=RigVMVariableNode() {variable=deform_value}.Value.bicep_44_#7
    weights=@n8.Result
    ExecuteContext=@Branch_7_RigVMUnitNode.True
    index=@GetMirrorItems_FNC_1_2_2_2_1.OutIndex
   )
   Branch_1_RigVMUnitNode = &RigVMFunction_ControlFlowBranch(Condition=RigVMVariableNode() {variable=deform_bool}.Value.knee_9_#8, ExecuteContext=@n2.B)
   n9 = &RigVMFunction_ForLoopCount(Count=2, ExecutePin=@Branch_1_RigVMUnitNode.True)
   > GetMirrorItems_FNC_1_2_2_2_1_1 = RigVMFunctionReferenceNode(Items=((Type=Bone,Name="calf_l"),(Type=Bone,Name="thigh_l")), l_side_token=_l, r_side_token=_r, Index=@n9.Index)
   > GetDeltaTransformFromRefPose_FNC_1_1 = RigVMFunctionReferenceNode(Item=RigVMDispatchNode(Index=0, Array=@GetMirrorItems_FNC_1_2_2_2_1_1.OutItems) {template="DISPATCH_RigVMDispatch_ArrayGetAtIndex(in Array,in Index,out Element)"}.Element)
   n10 = &RigUnit_SequenceExecution(ExecuteContext=@GetDeltaTransformFromRefPose_FNC_1_1.ExecuteContext)
   TwoParentConstraint_FNC_3 = RigVMFunctionReferenceNode(
    alpha=1
    l_side_token=_l
    masters=((Type=Bone,Name="calf_l"),(Type=Bone,Name="thigh_l"))
    r_side_token=_r
    slave=(Type=Bone,Name="deform_knee_in_base_l")
    weights=(0.5,0.5)
    ExecuteContext=@n10.A
    Index=@GetMirrorItems_FNC_1_2_2_2_1_1.OutIndex
   )
   n11 = &RigVMFunction_MathFloatRemap(SourceMaximum=100, SourceMinimum=10, bClamp=true, Value=@GetDeltaTransformFromRefPose_FNC_1_1.OutTransform.Rotation.Yaw)
   > SetDrivenKey_FNC_3 = RigVMFunctionReferenceNode(
    item=(Type=Bone,Name="deform_knee_in_l")
    l_side_token=_l
    r_side_token=_r
    index=@TwoParentConstraint_FNC_3.outIndex
    transform=RigVMVariableNode() {variable=deform_value}.Value.knee_in_41_#9
    weights=@n11.Result
   )
   TwoParentConstraint_FNC_3_1 = RigVMFunctionReferenceNode(
    alpha=1
    l_side_token=_l
    masters=((Type=Bone,Name="calf_l"),(Type=Bone,Name="thigh_l"))
    r_side_token=_r
    slave=(Type=Bone,Name="deform_knee_out_base_l")
    weights=(0.1,0.9)
    ExecuteContext=@n10.B
    Index=@GetMirrorItems_FNC_1_2_2_2_1_1.OutIndex
   )
   > SetDrivenKey_FNC_3_1 = RigVMFunctionReferenceNode(
    item=(Type=Bone,Name="deform_knee_out_l")
    l_side_token=_l
    r_side_token=_r
    index=@TwoParentConstraint_FNC_3_1.outIndex
    transform=RigVMVariableNode() {variable=deform_value}.Value.knee_out_42_#10
    weights=@n11.Result
   )
   Branch_2_RigVMUnitNode = &RigVMFunction_ControlFlowBranch(Condition=RigVMVariableNode() {variable=deform_bool}.Value.pec_12_#11, ExecuteContext=@n2.C)
   n12 = &RigVMFunction_ForLoopCount(Count=2, ExecutePin=@Branch_2_RigVMUnitNode.True)
   > GetMirrorItems_FNC_2_1_3 = RigVMFunctionReferenceNode(
    Items=((Type=Bone,Name="deform_pec_aim_base_l"),(Type=Bone,Name="spine_03"),(Type=Bone,Name="spine_04"),(Name="spine_05"))
    l_side_token=_l
    r_side_token=_r
    Index=@n12.Index
   )
   > &RigUnit_ParentConstraint(
    AdvancedSettings=(RotationOrderForFilter=XZY)
    Parents=((Item=(Type=Bone),Weight=0.25),(Item=(Type=Bone),Weight=0.55),(Item=(Type=Bone),Weight=0.2))
    Weight=1
    Child=RigVMDispatchNode(Index=0, Array=@GetMirrorItems_FNC_2_1_3.OutItems) {template="DISPATCH_RigVMDispatch_ArrayGetAtIndex(in Array,in Index,out Element)"}.Element
    Parents.0.Item=RigVMDispatchNode(Index=1, Array=@GetMirrorItems_FNC_2_1_3.OutItems) {template="DISPATCH_RigVMDispatch_ArrayGetAtIndex(in Array,in Index,out Element)"}.Element
    Parents.1.Item=RigVMDispatchNode(Index=2, Array=@GetMirrorItems_FNC_2_1_3.OutItems) {template="DISPATCH_RigVMDispatch_ArrayGetAtIndex(in Array,in Index,out Element)"}.Element
    Parents.2.Item=RigVMDispatchNode(Index=3, Array=@GetMirrorItems_FNC_2_1_3.OutItems) {template="DISPATCH_RigVMDispatch_ArrayGetAtIndex(in Array,in Index,out Element)"}.Element
   )
   > GetMirrorItems_FNC_3 = RigVMFunctionReferenceNode(
    Items=((Type=Bone,Name="deform_pec_aim_l"),(Type=Bone,Name="upperarm_l"))
    l_side_token=_l
    r_side_token=_r
    Index=@GetMirrorItems_FNC_2_1_3.OutIndex
   )
   n13 = RigVMDispatchNode(Index=0, Array=@GetMirrorItems_FNC_3.OutItems) {template="DISPATCH_RigVMDispatch_ArrayGetAtIndex(in Array,in Index,out Element)"}
   n14 = /Script/ControlRig.RigUnit_AimBoneMath(
    DebugSettings=(Scale=10,WorldOffset=(Rotation=(X=0,Y=0,Z=0,W=1),Translation=(X=0,Y=0,Z=0),Scale3D=(X=1,Y=1,Z=1)))
    Primary=(Weight=1,Axis=(X=1,Y=0,Z=0),Target=(X=0,Y=0,Z=0),Kind=Location,Space=(Type=Bone))
    Secondary=(Weight=1,Axis=(X=0,Y=-1,Z=0),Target=(X=0,Y=0,Z=0),Kind=Location,Space=(Type=Bone,Name="spine_04"))
    Weight=1
    InputTransform=&RigUnit_GetTransform(Item=@n13.Element).Transform
    Primary.Axis=RigVMDispatchNode(Values=((X=1,Y=0,Z=0),(X=-1,Y=0,Z=0)), Index=@GetMirrorItems_FNC_3.OutIndex) {template="DISPATCH_RigVMDispatch_SelectInt32(in Index,in Values,out Result)"}.Result
    Secondary.Axis=RigVMDispatchNode(Values=((X=0,Y=-1,Z=0),(X=0,Y=1,Z=0)), Index=@GetMirrorItems_FNC_3.OutIndex) {template="DISPATCH_RigVMDispatch_SelectInt32(in Index,in Values,out Result)"}.Result
    Primary.Space=RigVMDispatchNode(Index=1, Array=@GetMirrorItems_FNC_3.OutItems) {template="DISPATCH_RigVMDispatch_ArrayGetAtIndex(in Array,in Index,out Element)"}.Element
   ) {is=struct RigVMUnitNode}
   > &RigUnit_SetTransform(Weight=1, bInitial=false, Value=@n14.Result, Item=@n13.Element)
   > GetMirrorItems_FNC_2_1_3_1 = RigVMFunctionReferenceNode(
    Items=((Type=Bone,Name="deform_pec_l"),(Type=Bone,Name="deform_pec_base_l"),(Type=Bone,Name="upperarm_l"))
    l_side_token=_l
    r_side_token=_r
    Index=@GetMirrorItems_FNC_3.OutIndex
   )
   > /Script/ControlRig.RigUnit_PositionConstraint(
    Filter=(bX=True,bY=false,bZ=false)
    Parents=((Item=(Type=Bone),Weight=0),(Item=(Type=Bone),Weight=1))
    Weight=1
    bMaintainOffset=True
    Child=RigVMDispatchNode(Index=0, Array=@GetMirrorItems_FNC_2_1_3_1.OutItems) {template="DISPATCH_RigVMDispatch_ArrayGetAtIndex(in Array,in Index,out Element)"}.Element
    Parents.0.Item=RigVMDispatchNode(Index=1, Array=@GetMirrorItems_FNC_2_1_3_1.OutItems) {template="DISPATCH_RigVMDispatch_ArrayGetAtIndex(in Array,in Index,out Element)"}.Element
    Parents.1.Item=RigVMDispatchNode(Index=2, Array=@GetMirrorItems_FNC_2_1_3_1.OutItems) {template="DISPATCH_RigVMDispatch_ArrayGetAtIndex(in Array,in Index,out Element)"}.Element
   ) {is=struct RigVMUnitNode}
   > GetMirrorItems_FNC_0 = RigVMFunctionReferenceNode(Items=((Type=Bone,Name="clavicle_l")), l_side_token=_l, r_side_token=_r, Index=@GetMirrorItems_FNC_2_1_3_1.OutIndex)
   > GetDeltaTransformFromRefPose_FNC_0 = RigVMFunctionReferenceNode(Item=RigVMDispatchNode(Index=0, Array=@GetMirrorItems_FNC_0.OutItems) {template="DISPATCH_RigVMDispatch_ArrayGetAtIndex(in Array,in Index,out Element)"}.Element)
   > SetDrivenKey_FNC_0 = RigVMFunctionReferenceNode(
    item=(Type=Bone,Name="deform_pec_l")
    l_side_token=_l
    r_side_token=_r
    weights=&RigVMFunction_MathFloatRemap(SourceMaximum=30, bClamp=true, Value=@GetDeltaTransformFromRefPose_FNC_0.OutTransform.Rotation.Pitch).Result
    transform=RigVMVariableNode() {variable=deform_value}.Value.pec_51_#12
    index=@GetMirrorItems_FNC_0.OutIndex
   )
   n15 = &RigUnit_SequenceExecution(ExecuteContext=@n1.B)
   n16 = RigVMVariableNode() {variable=deform_bool}
   Branch_3_RigVMUnitNode = &RigVMFunction_ControlFlowBranch(
    Condition=&RigVMFunction_MathBoolOr(B=@n16.Value.glute_17_#13, A=@n16.Value.groin_8_#14).Result
    ExecuteContext=@n15.A
   )
   n17 = &RigVMFunction_ForLoopCount(Count=2, ExecutePin=@Branch_3_RigVMUnitNode.True)
   > PoseReader_FNC = RigVMFunctionReferenceNode(
    FalloffSize=0.45
    activeRegionSize=0.1
    drawDebug=true
    driver=(Type=Bone,Name="thigh_l")
    driverAxis=(X=-1,Y=0,Z=0)
    flipWidthScaling=false
    l_side_token=_l
    negativeHeight=3
    negativeWidth=0.5
    parentSpace=(Type=Bone)
    positiveHeight=3
    positiveWidth=3.5
    r_side_token=_r
    rotationOffset=((X=90,Y=0,Z=0),(X=-90,Y=180,Z=0))
    index=@n17.Index
    flipHeightScaling=&RigVMFunction_MathIntEquals(B=1, A=@n17.Index).Result
   )
   n18 = &RigUnit_SequenceExecution(ExecuteContext=@PoseReader_FNC.ExecuteContext)
   Branch_8_RigVMUnitNode = &RigVMFunction_ControlFlowBranch(ExecuteContext=@n18.A, Condition=RigVMVariableNode() {variable=deform_bool}.Value.groin_8_#14)
   TwoParentConstraint_FNC = RigVMFunctionReferenceNode(
    alpha=1
    l_side_token=_l
    masters=((Type=Bone,Name="thigh_l"),(Type=Bone,Name="pelvis"))
    r_side_token=_r
    slave=(Type=Bone,Name="deform_groin_base_l")
    weights=(0.7,0.3)
    ExecuteContext=@Branch_8_RigVMUnitNode.True
    Index=@PoseReader_FNC.outIndex
   )
   > SetDrivenKey_FNC = RigVMFunctionReferenceNode(
    item=(Type=Bone,Name="deform_groin_l")
    l_side_token=_l
    r_side_token=_r
    index=@TwoParentConstraint_FNC.outIndex
    transform=RigVMVariableNode() {variable=deform_value}.Value.groin_35_#15
    weights=@PoseReader_FNC.outWeights
   )
   Branch_8_1_RigVMUnitNode = &RigVMFunction_ControlFlowBranch(ExecuteContext=@n18.B, Condition=RigVMVariableNode() {variable=deform_bool}.Value.glute_17_#13)
   TwoParentConstraint_FNC_1 = RigVMFunctionReferenceNode(
    alpha=1
    l_side_token=_l
    masters=((Type=Bone,Name="thigh_l"),(Type=Bone,Name="pelvis"))
    r_side_token=_r
    slave=(Type=Bone,Name="deform_glute_base_l")
    weights=(0.25,0.75)
    ExecuteContext=@Branch_8_1_RigVMUnitNode.True
    Index=@PoseReader_FNC.outIndex
   )
   > SetDrivenKey_FNC_1 = RigVMFunctionReferenceNode(
    item=(Type=Bone,Name="deform_glute_l")
    l_side_token=_l
    r_side_token=_r
    index=@TwoParentConstraint_FNC_1.outIndex
    transform=RigVMVariableNode() {variable=deform_value}.Value.glute_36_#16
    weights=@PoseReader_FNC.outWeights
   )
   Branch_5_RigVMUnitNode = &RigVMFunction_ControlFlowBranch(Condition=RigVMVariableNode() {variable=deform_bool}.Value.upperarm_1_#17, ExecuteContext=@n15.B)
   n19 = &RigVMFunction_ForLoopCount(Count=2, ExecutePin=@Branch_5_RigVMUnitNode.True)
   > TwoParentConstraint_FNC_2 = RigVMFunctionReferenceNode(
    alpha=1
    l_side_token=_l
    masters=((Type=Bone,Name="upperarm_twist_01_l"),(Type=Bone,Name="clavicle_l"))
    r_side_token=_r
    slave=(Type=Bone,Name="deform_upperarm_up_base_l")
    weights=(0.5,0.5)
    Index=@n19.Index
   )
   > TwoParentConstraint_FNC_2_1 = RigVMFunctionReferenceNode(
    alpha=1
    l_side_token=_l
    masters=((Type=Bone,Name="upperarm_twist_01_l"),(Type=Bone,Name="clavicle_l"))
    r_side_token=_r
    slave=(Type=Bone,Name="deform_upperarm_fr_base_l")
    weights=(0.3,0.7)
    Index=@TwoParentConstraint_FNC_2.outIndex
   )
   > TwoParentConstraint_FNC_2_1_1 = RigVMFunctionReferenceNode(
    alpha=1
    l_side_token=_l
    masters=((Type=Bone,Name="upperarm_twist_01_l"),(Type=Bone,Name="clavicle_l"))
    r_side_token=_r
    slave=(Type=Bone,Name="deform_upperarm_bk_base_l")
    weights=(0.3,0.7)
    Index=@TwoParentConstraint_FNC_2_1.outIndex
   )
   > PoseReader_FNC_1 = RigVMFunctionReferenceNode(
    FalloffSize=0.5
    activeRegionSize=0.1
    drawDebug=false
    driver=(Type=Bone,Name="upperarm_l")
    driverAxis=(X=1,Y=0,Z=0)
    flipWidthScaling=false
    l_side_token=_l
    negativeHeight=3
    negativeWidth=3
    parentSpace=(Type=Bone)
    positiveHeight=3
    positiveWidth=2.5
    r_side_token=_r
    rotationOffset=((X=0,Y=0,Z=0),(X=180,Y=0,Z=180))
    flipHeightScaling=&RigVMFunction_MathIntEquals(B=1, A=@TwoParentConstraint_FNC_2_1_1.outIndex).Result
    index=@TwoParentConstraint_FNC_2_1_1.outIndex
   )
   > SetDrivenKey_FNC_2 = RigVMFunctionReferenceNode(
    item=(Type=Bone,Name="deform_upperarm_up_l")
    l_side_token=_l
    r_side_token=_r
    index=@PoseReader_FNC_1.outIndex
    weights=@PoseReader_FNC_1.outWeights
    transform=RigVMVariableNode() {variable=deform_value}.Value.upperarm_up_18_#18
   )
   > PoseReader_FNC_1_1 = RigVMFunctionReferenceNode(
    FalloffSize=0.4
    activeRegionSize=0.1
    drawDebug=false
    driver=(Type=Bone,Name="upperarm_l")
    driverAxis=(X=1,Y=0,Z=0)
    l_side_token=_l
    negativeHeight=1.5
    negativeWidth=3
    parentSpace=(Type=Bone)
    positiveHeight=1.5
    positiveWidth=0.5
    r_side_token=_r
    rotationOffset=((X=90,Y=0,Z=0),(X=90,Y=180,Z=180))
    flipWidthScaling=&RigVMFunction_MathIntEquals(B=1, A=@SetDrivenKey_FNC_2.outIndex).Result
    index=@SetDrivenKey_FNC_2.outIndex
   )
   > SetDrivenKey_FNC_2_1 = RigVMFunctionReferenceNode(
    item=(Type=Bone,Name="deform_upperarm_fr_l")
    l_side_token=_l
    r_side_token=_r
    index=@PoseReader_FNC_1_1.outIndex
    weights=@PoseReader_FNC_1_1.outWeights
    transform=RigVMVariableNode() {variable=deform_value}.Value.upperarm_fwd_fr_19_#19
   )
   > SetDrivenKey_FNC_2_1_1 = RigVMFunctionReferenceNode(
    item=(Type=Bone,Name="deform_upperarm_bk_l")
    l_side_token=_l
    r_side_token=_r
    index=@SetDrivenKey_FNC_2_1.outIndex
    weights=@SetDrivenKey_FNC_2_1.outWeights
    transform=RigVMVariableNode() {variable=deform_value}.Value.upperarm_fwd_bk_20_#20
   )
   > PoseReader_FNC_1_1_1 = RigVMFunctionReferenceNode(
    FalloffSize=0.4
    activeRegionSize=0.1
    drawDebug=false
    driver=(Type=Bone,Name="upperarm_l")
    driverAxis=(X=1,Y=0,Z=0)
    flipWidthScaling=false
    l_side_token=_l
    negativeHeight=1.5
    negativeWidth=1
    parentSpace=(Type=Bone)
    positiveHeight=1.5
    positiveWidth=3
    r_side_token=_r
    rotationOffset=((X=90,Y=0,Z=180),(X=90,Y=0,Z=0))
    flipHeightScaling=&RigVMFunction_MathIntEquals(B=1).Result
    index=@SetDrivenKey_FNC_2_1_1.outIndex
   )
   > SetDrivenKey_FNC_2_1_2 = RigVMFunctionReferenceNode(
    item=(Type=Bone,Name="deform_upperarm_bk_l")
    l_side_token=_l
    r_side_token=_r
    index=@PoseReader_FNC_1_1_1.outIndex
    weights=@PoseReader_FNC_1_1_1.outWeights
    transform=RigVMVariableNode() {variable=deform_value}.Value.upperarm_bk_22_#21
   )
   Branch_5_1_RigVMUnitNode = &RigVMFunction_ControlFlowBranch(Condition=RigVMVariableNode() {variable=deform_bool}.Value.wrist_7_#22, ExecuteContext=@n15.C)
   n20 = &RigVMFunction_ForLoopCount(Count=2, ExecutePin=@Branch_5_1_RigVMUnitNode.True)
   > TwoParentConstraint_FNC_2_2 = RigVMFunctionReferenceNode(
    alpha=1
    l_side_token=_l
    masters=((Type=Bone,Name="hand_l"),(Type=Bone,Name="lowerarm_twist_01_l"))
    r_side_token=_r
    slave=(Type=Bone,Name="deform_wrist_up_base_l")
    weights=(0.4,0.6)
    Index=@n20.Index
   )
   > TwoParentConstraint_FNC_2_2_1 = RigVMFunctionReferenceNode(
    alpha=1
    l_side_token=_l
    masters=((Type=Bone,Name="hand_l"),(Type=Bone,Name="lowerarm_twist_01_l"))
    r_side_token=_r
    slave=(Type=Bone,Name="deform_wrist_dn_base_l")
    weights=(0.4,0.6)
    Index=@TwoParentConstraint_FNC_2_2.outIndex
   )
   > PoseReader_FNC_1_1_1_1 = RigVMFunctionReferenceNode(
    FalloffSize=0.45
    activeRegionSize=0.1
    drawDebug=false
    driver=(Type=Bone,Name="hand_l")
    driverAxis=(X=1,Y=0,Z=0)
    flipWidthScaling=false
    l_side_token=_l
    negativeHeight=3
    negativeWidth=3.5
    parentSpace=(Type=Bone,Name="lowerarm_twist_01_l")
    positiveHeight=3
    positiveWidth=0.5
    r_side_token=_r
    rotationOffset=((X=90,Y=0,Z=0),(X=90,Y=0,Z=180))
    flipHeightScaling=&RigVMFunction_MathIntEquals(B=1, A=@TwoParentConstraint_FNC_2_2_1.outIndex).Result
    index=@TwoParentConstraint_FNC_2_2_1.outIndex
   )
   > SetDrivenKey_FNC_2_1_2_1 = RigVMFunctionReferenceNode(
    item=(Type=Bone,Name="deform_wrist_up_l")
    l_side_token=_l
    r_side_token=_r
    index=@PoseReader_FNC_1_1_1_1.outIndex
    weights=@PoseReader_FNC_1_1_1_1.outWeights
    transform=RigVMVariableNode() {variable=deform_value}.Value.wrist_up_31_#23
   )
   > PoseReader_FNC_1_1_1_1_1 = RigVMFunctionReferenceNode(
    FalloffSize=0.4
    activeRegionSize=0.1
    drawDebug=false
    driver=(Type=Bone,Name="hand_l")
    driverAxis=(X=1,Y=0,Z=0)
    flipWidthScaling=false
    l_side_token=_l
    negativeHeight=3
    negativeWidth=4
    parentSpace=(Type=Bone,Name="lowerarm_twist_01_l")
    positiveHeight=3
    positiveWidth=1
    r_side_token=_r
    rotationOffset=((X=-90,Y=0,Z=10),(X=-90,Y=0,Z=190))
    flipHeightScaling=&RigVMFunction_MathIntEquals(B=1, A=@SetDrivenKey_FNC_2_1_2_1.outIndex).Result
    index=@SetDrivenKey_FNC_2_1_2_1.outIndex
   )
   > SetDrivenKey_FNC_2_1_2_1_1 = RigVMFunctionReferenceNode(
    item=(Type=Bone,Name="deform_wrist_dn_l")
    l_side_token=_l
    r_side_token=_r
    index=@PoseReader_FNC_1_1_1_1_1.outIndex
    weights=@PoseReader_FNC_1_1_1_1_1.outWeights
    transform=RigVMVariableNode() {variable=deform_value}.Value.wrist_dn_32_#24
   )
  ]}
]
hierarchy [
 root {loc=<0 0 0>, rot=<0 0 0 1>}
 pelvis {parent=root, loc=<0 -0.7071 93.1191>, rot=<0.0127 0.707 -0.0127 -0.707>}
 spine_01 {parent=pelvis, loc=<6.7494 -0.6388 0>, rot=<0 0 0.034 -0.9994>}
 spine_02 {parent=spine_01, loc=<7.2518 -0.035 0>, rot=<0 0 0.0238 -0.9997>}
 spine_03 {parent=spine_02, loc=<6.0104 0 0>, rot=<0 0 0.0149 -0.9999>}
 spine_04 {parent=spine_03, loc=<8.1024 0 0>, rot=<0 0 -0.0132 -0.9999>}
 spine_05 {parent=spine_04, loc=<10.0146 0 0>, rot=<0 0 -0.0015 -1>}
 clavicle_l {parent=spine_05, loc=<4.3575 -0.1466 -2.3223>, rot=<0.0465 -0.7289 -0.009 -0.6829>}
 upperarm_l {parent=clavicle_l, loc=<15.5111 0 0>, rot=<0.0226 -0.3234 -0.0088 -0.9459>}
 lowerarm_l {parent=upperarm_l, loc=<26.6514 0 0>, rot=<0.0391 -0.0099 -0.2453 -0.9686>}
 hand_l {parent=lowerarm_l, loc=<22.3491 0 0>, rot=<-0.7081 -0.0507 0.0975 -0.6975>}
 index_metacarpal_l {parent=hand_l, loc=<1.6535 0.0183 -2.96>, rot=<0.0158 -0.0124 -0.0173 -0.9996>}
 index_01_l {parent=index_metacarpal_l, loc=<8.3008 0.0357 -0.0524>, rot=<-0.0044 0.0485 0.0924 -0.9945>}
 index_02_l {parent=index_01_l, loc=<3.3587 -0.3004 0.0452>, rot=<-0.002 0.0027 0.3225 -0.9466>}
 index_03_l {parent=index_02_l, loc=<2.5578 -0.123 -0.0252>, rot=<-0.002 0.0067 0.2558 -0.9667>}
 middle_metacarpal_l {parent=hand_l, loc=<2.1856 -0.1725 -0.5984>, rot=<0.006 0.0168 -0.0329 -0.9993>}
 middle_01_l {parent=middle_metacarpal_l, loc=<7.3275 -0.2447 0.0129>, rot=<-0.0208 0.0042 0.0858 -0.9961>}
 middle_02_l {parent=middle_01_l, loc=<4.0123 0.1401 0.1175>, rot=<-0.0218 0.0258 0.3512 -0.9357>}
 middle_03_l {parent=middle_02_l, loc=<2.6077 0.0534 0.1373>, rot=<0.0082 0.0188 0.1708 -0.9851>}
 pinky_metacarpal_l {parent=hand_l, loc=<1.7616 -0.9753 3.3814>, rot=<-0.0397 0.0814 -0.069 -0.9935>}
 pinky_01_l {parent=pinky_metacarpal_l, loc=<7.1709 -0.4831 -0.1094>, rot=<-0.0654 -0.0035 0.2242 -0.9723>}
 pinky_02_l {parent=pinky_01_l, loc=<2.8274 0.0893 -0.1043>, rot=<0.0037 0.0011 0.2854 -0.9584>}
 pinky_03_l {parent=pinky_02_l, loc=<2.2046 0.0307 -0.0078>, rot=<-0.0002 0.0098 0.1445 -0.9895>}
 ring_metacarpal_l {parent=hand_l, loc=<2.2074 -0.3625 1.3638>, rot=<-0.0284 0.0722 -0.0458 -0.9959>}
 ring_01_l {parent=ring_metacarpal_l, loc=<7.0295 -0.6197 -0.1446>, rot=<-0.0489 -0.0252 0.1892 -0.9804>}
 ring_02_l {parent=ring_01_l, loc=<3.6174 0.4647 -0.1027>, rot=<0 0 0.2806 -0.9598>}
 ring_03_l {parent=ring_02_l, loc=<2.4152 0.114 0.0199>, rot=<0 0 0.1189 -0.9929>}
 thumb_01_l {parent=hand_l, loc=<1.3118 -0.3191 -2.459>, rot=<0.6546 -0.4042 0.0664 -0.6354>}
 thumb_02_l {parent=thumb_01_l, loc=<5.9336 -0.0448 0.0401>, rot=<-0.0037 0.0213 0.298 -0.9543>}
 thumb_03_l {parent=thumb_02_l, loc=<2.2223 0.3682 -0.0907>, rot=<0.0118 -0.013 0.0066 -0.9998>}
 deform_wrist_dn_base_l {parent=hand_l, loc=<0 -2.2342 0>, rot=<0 0 0 1>}
 deform_wrist_dn_l {parent=deform_wrist_dn_base_l, loc=<0 0 0>, rot=<0 0 0 1>}
 deform_wrist_up_base_l {parent=hand_l, loc=<0 2.2682 0>, rot=<0 0 0 1>}
 deform_wrist_up_l {parent=deform_wrist_up_base_l, loc=<0 0 0>, rot=<0 0 0 1>}
 lowerarm_twist_01_l {parent=lowerarm_l, loc=<17.8549 0 0>, rot=<0 0 0 1>}
 lowerarm_twist_02_l {parent=lowerarm_l, loc=<9.6199 0 0>, rot=<0 0 0 1>}
 deform_elbow_in_base_l {parent=lowerarm_l, loc=<0.2112 5.259 0.8274>, rot=<0 0 0 1>}
 deform_elbow_in_l {parent=deform_elbow_in_base_l, loc=<0 0 0>, rot=<0 0 0 1>}
 deform_elbow_out_base_l {parent=lowerarm_l, loc=<0 -4 0>, rot=<0 0 0 1>}
 deform_elbow_out_l {parent=deform_elbow_out_base_l, loc=<0 0 0>, rot=<0 0 0 1>}
 upperarm_twist_01_l {parent=upperarm_l, loc=<0.0041 0 0>, rot=<0 0 0 1>}
 deform_upperarm_bk_base_l {parent=upperarm_twist_01_l, loc=<1.0813 -7.2193 -1.9453>, rot=<0 0 0 1>}
 deform_upperarm_bk_l {parent=deform_upperarm_bk_base_l, loc=<0 0 0>, rot=<0 0 0 1>}
 deform_upperarm_fr_base_l {parent=upperarm_twist_01_l, loc=<-0.3691 7.0506 -2.765>, rot=<0 0 0 1>}
 deform_upperarm_fr_l {parent=deform_upperarm_fr_base_l, loc=<0 0 0>, rot=<0 0 0 1>}
 deform_upperarm_up_base_l {parent=upperarm_twist_01_l, loc=<-3.2216 -1.1391 3.3311>, rot=<0 0 0 1>}
 deform_upperarm_up_l {parent=deform_upperarm_up_base_l, loc=<0 0 0>, rot=<0 0 0 1>}
 upperarm_twist_02_l {parent=upperarm_l, loc=<13.7941 0 0>, rot=<0 0 0 1>}
 clavicle_r {parent=spine_05, loc=<4.3575 -0.1466 2.3223>, rot=<0.7289 0.0465 -0.6829 0.009>}
 upperarm_r {parent=clavicle_r, loc=<-15.5111 0 0>, rot=<0.0226 -0.3234 -0.0088 -0.9459>}
 lowerarm_r {parent=upperarm_r, loc=<-26.6515 0 0>, rot=<0.0391 -0.0099 -0.2453 -0.9686>}
 hand_r {parent=lowerarm_r, loc=<-22.3492 0 0>, rot=<-0.7081 -0.0507 0.0975 -0.6975>}
 index_metacarpal_r {parent=hand_r, loc=<-1.6534 -0.0181 2.9601>, rot=<0.0158 -0.0124 -0.0173 -0.9996>}
 index_01_r {parent=index_metacarpal_r, loc=<-8.3008 -0.0357 0.0524>, rot=<-0.0044 0.0485 0.0924 -0.9945>}
 index_02_r {parent=index_01_r, loc=<-3.3589 0.3003 -0.0453>, rot=<-0.002 0.0027 0.3225 -0.9466>}
 index_03_r {parent=index_02_r, loc=<-2.5578 0.123 0.0252>, rot=<-0.002 0.0067 0.2558 -0.9667>}
 middle_metacarpal_r {parent=hand_r, loc=<-2.1856 0.1726 0.5984>, rot=<0.006 0.0168 -0.0329 -0.9993>}
 middle_01_r {parent=middle_metacarpal_r, loc=<-7.3273 0.2447 -0.0129>, rot=<-0.0208 0.0042 0.0858 -0.9961>}
 middle_02_r {parent=middle_01_r, loc=<-4.0142 -0.1411 -0.07>, rot=<-0.0217 0.0259 0.3512 -0.9357>}
 middle_03_r {parent=middle_02_r, loc=<-2.6074 -0.0534 -0.1312>, rot=<0.0083 0.0188 0.1708 -0.9851>}
 pinky_metacarpal_r {parent=hand_r, loc=<-1.7617 0.9753 -3.3814>, rot=<-0.0397 0.0814 -0.069 -0.9935>}
 pinky_01_r {parent=pinky_metacarpal_r, loc=<-7.1707 0.4832 0.1094>, rot=<-0.0654 -0.0035 0.2242 -0.9723>}
 pinky_02_r {parent=pinky_01_r, loc=<-2.8274 -0.0892 0.1043>, rot=<0.0037 0.0011 0.2854 -0.9584>}
 pinky_03_r {parent=pinky_02_r, loc=<-2.2045 -0.031 0.0078>, rot=<-0.0002 0.0098 0.1445 -0.9895>}
 ring_metacarpal_r {parent=hand_r, loc=<-2.2073 0.3627 -1.3638>, rot=<-0.0284 0.0722 -0.0458 -0.9959>}
 ring_01_r {parent=ring_metacarpal_r, loc=<-7.0297 0.6197 0.1446>, rot=<-0.0489 -0.0252 0.1892 -0.9804>}
 ring_02_r {parent=ring_01_r, loc=<-3.6171 -0.4648 0.1027>, rot=<0 0 0.2806 -0.9598>}
 ring_03_r {parent=ring_02_r, loc=<-2.4153 -0.1141 -0.0199>, rot=<0 0 0.1189 -0.9929>}
 thumb_01_r {parent=hand_r, loc=<-1.3119 0.3192 2.459>, rot=<0.6546 -0.4042 0.0664 -0.6354>}
 thumb_02_r {parent=thumb_01_r, loc=<-5.9335 0.0447 -0.0401>, rot=<-0.0037 0.0213 0.298 -0.9543>}
 thumb_03_r {parent=thumb_02_r, loc=<-2.2223 -0.3682 0.0907>, rot=<0.0118 -0.013 0.0066 -0.9998>}
 deform_wrist_dn_base_r {parent=hand_r, loc=<-0.0003 2.2346 -0.0001>, rot=<0 0 0 1>}
 deform_wrist_dn_r {parent=deform_wrist_dn_base_r, loc=<0 0 0>, rot=<0 0 0 1>}
 deform_wrist_up_base_r {parent=hand_r, loc=<0.0004 -2.2683 0.0001>, rot=<0 0 0 1>}
 deform_wrist_up_r {parent=deform_wrist_up_base_r, loc=<0 0 0>, rot=<0 0 0 1>}
 lowerarm_twist_01_r {parent=lowerarm_r, loc=<-17.855 0 0>, rot=<0 0 0 1>}
 lowerarm_twist_02_r {parent=lowerarm_r, loc=<-9.6199 0 0>, rot=<0 0 0 1>}
 deform_elbow_in_base_r {parent=lowerarm_r, loc=<-0.2111 -5.2591 -0.8275>, rot=<0 0 0 1>}
 deform_elbow_in_r {parent=deform_elbow_in_base_r, loc=<-0.0001 0 0.0002>, rot=<0 0 0 1>}
 deform_elbow_out_base_r {parent=lowerarm_r, loc=<0 4 0>, rot=<0 0 0 1>}
 deform_elbow_out_r {parent=deform_elbow_out_base_r, loc=<0 0 0>, rot=<0 0 0 1>}
 upperarm_twist_01_r {parent=upperarm_r, loc=<-0.0041 0 0>, rot=<0 0 0 1>}
 deform_upperarm_bk_base_r {parent=upperarm_twist_01_r, loc=<-1.0815 7.2192 1.9456>, rot=<0 0 0 1>}
 deform_upperarm_bk_r {parent=deform_upperarm_bk_base_r, loc=<0 0 0>, rot=<0 0 0 1>}
 deform_upperarm_fr_base_r {parent=upperarm_twist_01_r, loc=<0.3689 -7.0506 2.7652>, rot=<0 0 0 1>}
 deform_upperarm_fr_r {parent=deform_upperarm_fr_base_r, loc=<0 0 0>, rot=<0 0 0 1>}
 deform_upperarm_up_base_r {parent=upperarm_twist_01_r, loc=<3.2213 1.1391 -3.3308>, rot=<0 0 0 1>}
 deform_upperarm_up_r {parent=deform_upperarm_up_base_r, loc=<0 0 0>, rot=<0 0 0 1>}
 upperarm_twist_02_r {parent=upperarm_r, loc=<-13.7942 0 0>, rot=<0 0 0 1>}
 neck_01 {parent=spine_05, loc=<11.3138 0.1529 0>, rot=<0 0 -0.2003 -0.9797>}
 neck_02 {parent=neck_01, loc=<4.5627 0 0>, rot=<0 0 0.0504 -0.9987>}
 head {parent=neck_02, loc=<5.3914 0 0>, rot=<0 0 0.115 -0.9934>}
 deform_pec_aim_base_r {parent=spine_04, loc=<7.8727 12.6388 2.4123>, rot=<-0.4602 -0.0102 0.7936 0.3979>}
 deform_pec_aim_r {parent=deform_pec_aim_base_r, loc=<0 0 0>, rot=<0 0 0 1>}
 deform_pec_base_r {parent=deform_pec_aim_r, loc=<-10.7662 -0.0006 -0.0007>, rot=<0 0 0 1>}
 deform_pec_r {parent=deform_pec_base_r, loc=<0 0 0>, rot=<0 0 0 1>}
 deform_pec_aim_base_l {parent=spine_04, loc=<7.8732 12.6388 -2.4123>, rot=<0.0102 -0.4602 0.3979 -0.7936>}
 deform_pec_aim_l {parent=deform_pec_aim_base_l, loc=<0 0 0>, rot=<0 0 0 1>}
 deform_pec_base_l {parent=deform_pec_aim_l, loc=<10.766 0 0>, rot=<0 0 0 1>}
 deform_pec_l {parent=deform_pec_base_l, loc=<0.055 0 0>, rot=<0 0 0 1>}
 thigh_l {parent=pelvis, loc=<-2.5693 -0.0109 -7.7501>, rot=<0.0012 0.0572 0.0155 -0.9982>}
 calf_l {parent=thigh_l, loc=<-39.0484 0 0>, rot=<0.0006 -0.0161 -0.0391 -0.9991>}
 calf_twist_01_l {parent=calf_l, loc=<-35.813 -0.0967 0.0397>, rot=<0.0237 0.0095 -0.0035 -0.9997>}
 calf_twist_02_l {parent=calf_l, loc=<-18.8064 -0.0967 0.0397>, rot=<0.0237 0.0095 -0.0035 -0.9997>}
 foot_l {parent=calf_l, loc=<-42.561 0.1077 0.4292>, rot=<0.0418 -0.0455 0.0343 -0.9975>}
 ball_l {parent=foot_l, loc=<-6.1599 13.8264 0.6806>, rot=<0.0001 -0.0001 -0.7063 -0.7079>}
 deform_knee_in_base_l {parent=calf_l, loc=<1.298 -5.2957 -0.0496>, rot=<0 0 0 1>}
 deform_knee_in_l {parent=deform_knee_in_base_l, loc=<0 0 0>, rot=<0 0 0 1>}
 deform_knee_out_base_l {parent=calf_l, loc=<0 4 0>, rot=<0 0 0 1>}
 deform_knee_out_l {parent=deform_knee_out_base_l, loc=<0 0 0>, rot=<0 0 0 1>}
 thigh_twist_01_l {parent=thigh_l, loc=<-21.6303 0 0>, rot=<0.0013 0 0 -1>}
 deform_groin_base_l {parent=thigh_l, loc=<-2.1908 8.1142 -1.7442>, rot=<0 0 0 1>}
 deform_groin_l {parent=deform_groin_base_l, loc=<0 0 0>, rot=<0 0 0 1>}
 deform_glute_base_l {parent=thigh_l, loc=<-6 -9 0>, rot=<0 0 -0.4617 -0.887>}
 deform_glute_l {parent=deform_glute_base_l, loc=<0 0 0>, rot=<0 0 0 1>}
 thigh_r {parent=pelvis, loc=<-2.5691 -0.0109 7.7501>, rot=<0.0572 -0.0012 0.9982 0.0155>}
 calf_r {parent=thigh_r, loc=<39.0487 0 0>, rot=<0.0006 -0.0161 -0.0391 -0.9991>}
 calf_twist_01_r {parent=calf_r, loc=<35.8132 0.0967 -0.0397>, rot=<0.0237 0.0095 -0.0035 -0.9997>}
 calf_twist_02_r {parent=calf_r, loc=<18.8064 0.0967 -0.0397>, rot=<0.0237 0.0095 -0.0035 -0.9997>}
 foot_r {parent=calf_r, loc=<42.5612 -0.1077 -0.4292>, rot=<0.0418 -0.0455 0.0343 -0.9975>}
 ball_r {parent=foot_r, loc=<6.1598 -13.8264 -0.6806>, rot=<0.0001 -0.0001 -0.7063 -0.7079>}
 deform_knee_in_base_r {parent=calf_r, loc=<-1.2981 5.2958 0.0496>, rot=<0 0 0 1>}
 deform_knee_in_r {parent=deform_knee_in_base_r, loc=<0 0 0>, rot=<0 0 0 1>}
 deform_knee_out_base_r {parent=calf_r, loc=<0 -4 0>, rot=<0 0 0 1>}
 deform_knee_out_r {parent=deform_knee_out_base_r, loc=<0 0 0>, rot=<0 0 0 1>}
 thigh_twist_01_r {parent=thigh_r, loc=<21.6303 0 0>, rot=<0.0013 0 0 -1>}
 deform_groin_base_r {parent=thigh_r, loc=<2.1909 -8.1142 1.7442>, rot=<0 0 0 1>}
 deform_groin_r {parent=deform_groin_base_r, loc=<0 0 0>, rot=<0 0 0 -1>}
 deform_glute_base_r {parent=thigh_r, loc=<6 9 0>, rot=<0 0 -0.4617 -0.887>}
 deform_glute_r {parent=deform_glute_base_r, loc=<0 0 0>, rot=<0 0 0 1>}
]
variables [
 deform_bool {
  type=FUserDefinedStruct_#25
  default="(upperarm_1_#17=True,elbow_3_#4=True,wrist_7_#22=True,groin_8_#14=True,glute_17_#13=True,knee_9_40C408324CA261 ...(truncated)"
 }
 deform_value {
  type=FUserDefinedStruct_#26
  default="(upperarm_up_18_#18=(Rotation=(X=0,Y=0,Z=0,W=1),Translation=(X=0,Y=0,Z=8),Scale3D=(X=0.7,Y=1,Z=1)),upperarm_fwd_fr_19_#19=(Rotation=(X=0,Y=0,Z=0,W=1),Translation=(X=0,Y=10,Z=0),Scale3D=(X=0.5,Y=1,Z= ...(truncated)"
 }
]
