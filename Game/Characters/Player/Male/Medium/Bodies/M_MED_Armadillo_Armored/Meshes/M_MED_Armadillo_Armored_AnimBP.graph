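# Animation Blueprint for the armored Armadillo medium male body, correcting the left and right dyn_upperarm bones.
# It copies the attached parent's pose, runs the deform Control Rig, and caches the result for two additive bone-offset chains.
# Skydiving selects fixed offsets.
# Otherwise, offsets are weighted by each upper arm's reference-transform delta from spine_05, computed each update, with the right arm remapping the absolute raw output.
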AnimBlueprint /Game/Characters/Player/Male/Medium/Bodies/M_MED_Armadillo_Armored/Meshes/M_MED_Armadillo_Armored_AnimBP
target_skeleton=&/M_MED_Armadillo_Armored_Skeleton
graphs [
 AnimGraph {nodes [
   n1 = "Transform (Modify) Bone - Bone: dyn_upperarm_l"(ComponentPose=Local To Component(LocalPose=Use cached pose 'ctrl rig'()), Alpha=1, Translation="5,0,0", Rotation="-35,0,0", Scale="1,1,1") {
    rotationMode=BMM_Additive
    translationMode=BMM_Additive
    alphaBoolBlend {blendOption=Linear}
    boneToModify {boneName=dyn_upperarm_l}
    rotation {pitch=-35}
    translation {x=5}
   }
   n2 = "Transform (Modify) Bone - Bone: dyn_upperarm_r"(ComponentPose=@n1, Alpha=1, Translation="-5,0,0", Rotation="-45,0,0", Scale="1,1,1") {
    rotationMode=BMM_Additive
    rotationSpace=BCS_BoneSpace
    translationMode=BMM_Additive
    translationSpace=BCS_BoneSpace
    alphaBoolBlend {blendOption=Linear}
    boneToModify {boneName=dyn_upperarm_r}
    rotation {pitch=-45}
    translation {x=-5}
   }
   n3 = "Transform (Modify) Bone - Bone: dyn_upperarm_l"(
    ComponentPose=Local To Component(LocalPose=Use cached pose 'ctrl rig'())
    Alpha=Get Arm L rot() -> (Arm L rot:double)
    Translation="3.5,0,2"
    Rotation="0,-30,30"
    Scale="1,1,1"
   ) {
    rotationMode=BMM_Additive
    translationMode=BMM_Additive
    alphaBoolBlend {blendOption=Linear}
    boneToModify {boneName=dyn_upperarm_l}
    rotation {roll=30, yaw=-30}
    translation {x=3.5, z=2}
   }
   n4 = "Transform (Modify) Bone - Bone: dyn_upperarm_r"(ComponentPose=@n3, Alpha=Get Arm R rot() -> (Arm R rot:double), Translation="3,5,0", Rotation="0,-35,0", Scale="1,1,1") {
    rotationMode=BMM_Additive
    rotationSpace=BCS_BoneSpace
    translationMode=BMM_Additive
    translationSpace=BCS_BoneSpace
    alphaBoolBlend {blendOption=Linear}
    boneToModify {boneName=dyn_upperarm_r}
    rotation {yaw=-35}
    translation=<3 5>
   }
   Output Pose(Result=Blend Poses by bool(BlendPose_0=Component To Local(ComponentPose=@n2), BlendPose_1=Component To Local(ComponentPose=@n4), bActiveValue=Get bIsSkyDiving()) {blendType=Linear}) {layerGroup=None}
   n5 = Control Rig(Source=Copy Pose From Mesh() {bUseAttachedParent=true}) {
    alphaBoolBlend {blendOption=Linear}
    controlRigAssetReference {blueprintRigClass=&/M_MED_DeformRig_Armadillo_Armored_ControlRig.M_MED_DeformRig_Armadillo_Armored_ControlRig_C}
   }
   Save cached pose 'ctrl rig'(Pose=@n5) {cacheName=ctrl rig}
  ]}
 EventGraph {nodes [
   Event BlueprintUpdateAnimation()
   > Sequence()
   > n1 = Branch(Condition=Get bIsSkyDiving())
   > Set Arm L rot(Arm L rot:double=0) -> (Output_Get:double)
   > Set Arm R rot(Arm R rot:double=0) -> (Output_Get:double)
   n2 = Get Skel_Mesh()
   Set Arm L rot(
    execute=@n1.else
    Arm L rot=GetAlphaFromReferenceTransformDelta(SkeletalMeshComponent=@n2, BoneName=upperarm_l, BaseName=spine_05, InputRange=(X=40,Y=90), TwistAxis="1,0,0")
   )
   n3 = GetAlphaFromReferenceTransformDelta(SkeletalMeshComponent=@n2, BoneName=upperarm_r, BaseName=spine_05, InputRange=(X=60,Y=120), TwistAxis="1,0,0")
   > Set Arm R rot(Arm R rot=MapRangeClamped(Value="Absolute (Float)"(A=@n3.RawOutput), InRangeA=60, InRangeB=120))
   Event BlueprintInitializeAnimation()
   > Set Skel_Mesh(Skel_Mesh=GetOwnerSkeletalMeshComponent())
  ]}
]
variables [
 Arm L rot {type=number}
 Arm R rot {type=number}
 Skel_Mesh {type=SkeletalMeshComponent}
]
